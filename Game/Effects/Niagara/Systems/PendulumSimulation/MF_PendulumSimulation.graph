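# Material function that builds the world position offset for mesh particles in the Niagara pendulum simulation effect.
# Rotates the mesh around its particle position using a quaternion from Dynamic Parameter index 3, blending that offset from zero by UV V so it varies along the mesh.
# Feeds the offset position into SplineThicken, with base and tip widths as inputs, and adds both offsets into the WPO output.
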MaterialFunction /Game/Effects/Niagara/Systems/PendulumSimulation/MF_PendulumSimulation
graph {nodes [
  n1 = LinearInterpolate(
   B=MaterialFunctionCall!(
    Quaternion=DynamicParameter() {parameterIndex=3}.RGBA
    Invert Quaternion
    World Position=WorldPosition().XYZ
    Particle Position=ParticlePositionWS()
   ) {materialFunction=/Niagara/Functions/DataPacking/Niagara_RotateMeshParticleWithQuaternion}.Result
   Alpha=ComponentMask(None=TextureCoordinate()) {g=true}
  ) {constB=0}
  n2 = FunctionInput() {inputName=SplineThicken, inputType=FunctionInput_Scalar, previewValue=<10 0 0 0>}
  n3 = FunctionInput() {inputName=SplineTip, inputType=FunctionInput_Scalar, previewValue=<10 0 0 0>}
  n4 = Add(A=@n1, B=WorldPosition().XYZ)
  FunctionOutput(None=Add(
    A=@n1
    B=MaterialFunctionCall(WidthBase=@n2, WidthTip=@n3, WorldPosition=@n4) {materialFunction=&SplineThicken}.WorldPositionOffset
   )) {outputName=WPO}
 ]}
