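# Two-sided Default Lit material for Niagara mesh particles, the strip material of the PendulumSimulation effect.
# Its graph hands the work to the MF_PendulumSimulation material function, fed by two scalar parameters.
# The parameter names are crossed against the inputs: SplineThicken drives the SplineTip input and SplineTip drives SplineThicken.
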Material /Game/Effects/Niagara/Systems/PendulumSimulation/M_PendulumSimulation_Strip
graph {
 properties {bUsedWithNiagaraMeshParticles=true, pixelDepthOffsetMode=PDOM_Legacy, twoSided=true, usedShadingModels=Default Lit}
 nodes [MaterialFunctionCall!(
   SplineTip=ScalarParameter() {defaultValue=10, parameterName=SplineThicken}
   SplineThicken=ScalarParameter() {defaultValue=10, parameterName=SplineTip}
  ) {materialFunction=&/MF_PendulumSimulation}]
}
