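# Translucent, unlit, two-sided Niagara mesh-particle material for a pendulum simulation strip, named as a VFX preview variant.
# World position offset math adds MF_PendulumSimulation, with spline tip and thicken parameters, to a camera offset.
# A dynamic-parameter-timed panner builds UVs for a smoke trail texture, masked by a vertical gradient, with a detail-dust blend alongside.
# Channel masks drive two-tint emissive and depth-faded opacity, both scaled by particle color; a Fresnel term is computed per vertex.
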Material /Game/Effects/Niagara/Systems/PendulumSimulation/M_PendulumSimulation_Strip_VFXPreview
graph {
 properties {
  bUsedWithNiagaraMeshParticles=true
  blendMode=BLEND_Translucent
  pixelDepthOffsetMode=PDOM_Legacy
  shadingModel=MSM_Unlit
  twoSided=true
  usedShadingModels=Unlit
 }
 nodes [
  Add(
   A=MaterialFunctionCall!(
    SplineTip=ScalarParameter() {defaultValue=75, parameterName=SplineThicken}
    SplineThicken=ScalarParameter() {defaultValue=26.809, parameterName=SplineTip}
   ) {materialFunction=&/MF_PendulumSimulation}.WPO
   B=MaterialFunctionCall(Offset Amount=ScalarParameter() {parameterName=MaterialCamOffset}) {materialFunction=/Engine/Functions/Engine_MaterialFunctions02/WorldPositionOffset/CameraOffset}.Camera Offset
  )
  n1 = TextureSampleParameter2D(UVs=NamedRerouteUsage()) {group=Main, parameterName=MainTexture, texture=/Game/Effects/Fort_Effects/Textures/Smoke/T_Smoke_Trail_Soft_2}
  n2 = ChannelMaskParameter(None=@n1.RGBA) {group=Main, maskChannel=Green, parameterName=MainTexChannel, defaultValue=<rgba 0 1 0 0>}
  n3 = TextureSampleParameter2D(UVs=NamedRerouteUsage()) {group=Detail, parameterName=DetailTexture, texture=/Game/Global/Textures/Default/PointSplatter/T_Tiling_Dust}
  LinearInterpolate(
   A=@n2.R
   B=Multiply(
    A=@n2.R
    B=ChannelMaskParameter(None=@n3.RGBA) {group=Detail, maskChannel=Green, parameterName=DetailTexChannel, defaultValue=<rgba 0 1 0 0>}.R
   )
   Alpha=ScalarParameter() {defaultValue=0.7788, group=Main, parameterName="DetialTexBlend(Multiply)"}
  )
  n4 = TextureCoordinate()
  Multiply(A=@n4) {constB=-1}
  n5 = TextureCoordinate()
  n6 = MaterialFunctionCall(XY=@n5) {materialFunction=&Swizzle}
  n7 = ScalarParameter() {defaultValue=1, group=Main, parameterName=Main_UV_Swizzle}
  n8 = VectorParameter() {group=Main, parameterName=Main_UV, defaultValue=<rgba 0.485 1 -0.1627 0>}
  NamedRerouteDeclaration(None=MaterialFunctionCall!(
    ClampTexture
    UV=LinearInterpolate(A=@n5, B=@n6.YX, Alpha=@n7)
    Tex
    InRGB=@n8.RGB
    InA=@n8.RGB
    Time=DynamicParameter() {defaultValue=<rgba 1 1 1 250>, paramNames [Time,Distort,Dissolve,WPO_Width]}.Time
    TimeMult
   ) {materialFunction=/Game/Effects/Niagara/MasterMaterials/Functions/MF_FX_VectorDrive_TexPanner}.PannerUV) {name=MainUV, nodeColor=<rgba 0 0.4049 1 1>}
  NamedRerouteDeclaration(None=Multiply(A=@n1.RGBA, B=TextureSample(UVs=@n4) {texture=/Game/Effects/Fort_Effects/Textures/Masks/T_Vert_Gradient_Tilable}.G)) {name=blendedTexture, nodeColor=<rgba 0.0339 1 0 1>}
  n9 = VectorParameter() {parameterName=UV1, defaultValue=<rgba 0.2453 0.2582 -0.168 0.344>}
  NamedRerouteDeclaration(None=MaterialFunctionCall(
    UV=MaterialFunctionCall!(TextureResolution) {materialFunction=/Engine/Functions/Engine_MaterialFunctions02/Texturing/ScreenAlignedPixelToPixelUVs}.Result
    InRGB=@n9.RGB
    InA=@n9.RGB
    Time=DynamicParameter() {defaultValue=<rgba 1 1 1 250>, paramNames [Time,Distort,Dissolve,WPO_Width]}.Time
   ) {materialFunction=&MF_FX_VectorDrive_TexPanner}.PannerUV) {name=UV_1, nodeColor=<rgba 0 0.4049 1 1>}
  n10 = VectorParameter() {group=Emissive, parameterName=EmissiveColorTintA, defaultValue=<rgba 0.0474 0.0663 0.1319 1>}
  n11 = VectorParameter() {group=Emissive, parameterName=EmissiveColorTintB, defaultValue=<rgba 0.1736 0.5868 1 1>}
  n12 = ChannelMaskParameter(None=NamedRerouteUsage()) {group=Emissive, maskChannel=Green, parameterName=EmissiveChannel, defaultValue=<rgba 0 1 0 0>}
  NamedRerouteDeclaration(None=Multiply(A=ParticleColor().RGB, B=LinearInterpolate(A=@n10.RGB, B=@n11.RGB, Alpha=@n12.R))) {name=Emissive, nodeColor=<rgba 0 0 0 1>}
  n13 = ParticleColor()
  n14 = MaterialFunctionCall!(
   InAlpha=ChannelMaskParameter(None=NamedRerouteUsage()) {group=Opacity, maskChannel=Green, parameterName=OpacityChannel, defaultValue=<rgba 0 1 0 0>}.R
   OpacityBasedFade_DistanceA=ScalarParameter() {defaultValue=2, group=Opacity, parameterName=OpacityBasedFade_DistanceA}
   OpacityBasedFade_DistanceB=ScalarParameter() {group=Opacity, parameterName=OpacityBasedFade_DistanceB}
  ) {materialFunction=/Game/Effects/Niagara/MasterMaterials/Functions/MF_FX_OpacityBasedDepthFade}
  NamedRerouteDeclaration(None=Subtract(A=Multiply(A=@n13.A, B=@n14.FinalAlpha), B=ScalarParameter() {group=Opacity, parameterName="Opacity_Sub(DP)"})) {name=Opacity, nodeColor=<rgba 0 0 0 1>}
  NamedRerouteUsage()
  NamedRerouteUsage()
  n15 = Subtract(A=CameraPositionWS(), B=WorldPosition().XYZ)
  n16 = Abs(None=DotProduct(A=VertexNormalWS(), B=Normalize(VectorInput=@n15)))
  n17 = Power(
   Base=LinearInterpolate(A=OneMinus(None=@n16), B=@n16, Alpha=ScalarParameter() {defaultValue=1, parameterName="Fresnel_Invert (0or1)"})
   Exp=ScalarParameter() {defaultValue=2, parameterName=Fresnel_Power}
  )
  n18 = ScalarParameter() {defaultValue=0.05, parameterName=Fresnel_Offset}
  VertexInterpolator(VS=Multiply(A=Subtract(A=@n17, B=@n18), B=Add(A=@n18)))
 ]
}
