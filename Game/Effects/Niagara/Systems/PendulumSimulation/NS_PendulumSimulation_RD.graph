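# Niagara test system that simulates pendulum-like particle motion, drawn with sprite and mesh renderers.
# Two active emitters burst-spawn particles that outlive their lifetime and run NM_PendulumSimulation with a tail pivot and hardness, then Drag and the force solver.
# Empty005 also drives a dynamic material parameter from randomly scaled and offset engine time.
# Four disabled emitters hold apparent earlier iterations using local FollowUpUpdate scratch modules, the first adding noise, vortex and point forces.
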NiagaraSystem /Game/Effects/Niagara/Systems/PendulumSimulation/NS_PendulumSimulation_RD
dependencies {usedDataInterfaces [], usedDynamicInputs [], usedModules [&SystemState]}
emitters {
 Empty {
  bEnabled=false
  emitterUpdateScript [
   &EmitterState(Life Cycle Mode=::NewEnumerator1, Loop Behavior=::NewEnumerator1, Loop Duration=1, UseLoopDelay=false)
   &SpawnBurst_Instantaneous(Spawn Count=10, Use Spawn Probability=false, UseLoopCountLimit=false)
   &SpawnRate(SpawnRate=10, Use Spawn Probability=false) {enabled=false}
  ]
  particleSpawnScript [
   &InitializeParticle(
    Color Mode=::NewEnumerator0
    Lifetime=5
    Mesh Scale Mode=::NewEnumerator3
    Mesh Uniform Scale=0.4
    Sprite Size Mode=::NewEnumerator3
    Uniform Sprite Size=20
    UsePositionOffset=false
    Write Material Random=true
   )
   &ShapeLocation(MaskSpawnGroup=false, Override Seed=false, UseOwnerScale=true)
   /SetVariables(Particles.Color001:LinearColor=<rgba 1 0 0 1>) {isSetParameters=true, module=SetVariables_#1}
  ]
  particleUpdateScript [
   &ParticleState(Kill Particles When Lifetime Has Elapsed=false)
   &self:FollowUpUpdate(Dir:Vector3f=<200 0 0>) {module=FollowUpUpdate}
   &CurlNoiseForce(Pan Noise=false)
   &VortexForce(Origin Pull Amount=100, UseOriginOffset=false)
   &Drag(Use Linear Drag=true, Use Rotational Drag=true)
   &PointForce(Force Strength=-1297.3162, Module.Use Falloff=false, Module.UseLinearForce=false)
   &SolveForcesAndVelocity(Clamp Velocity=false, Limit Acceleration=false)
  ]
  rendererClasses [&NiagaraSpriteRendererProperties, &NiagaraMeshRendererProperties]
  renderers [
   {rendererClass=&NiagaraSpriteRendererProperties}
   {rendererClass=&NiagaraMeshRendererProperties, rendererIndex=1}
   {rendererClass=&NiagaraSpriteRendererProperties, rendererIndex=2}
  ]
 }
 Empty001 {
  bEnabled=false
  emitterUpdateScript [
   &EmitterState(Life Cycle Mode=::NewEnumerator1, Loop Behavior=::NewEnumerator1, Loop Duration=1, UseLoopDelay=false)
   &SpawnBurst_Instantaneous(Spawn Count=10, Use Spawn Probability=false, UseLoopCountLimit=false)
   &SpawnRate(SpawnRate=10, Use Spawn Probability=false) {enabled=false}
  ]
  particleSpawnScript [
   &InitializeParticle(
    Color Mode=::NewEnumerator0
    Lifetime=5
    Mesh Scale Mode=::NewEnumerator3
    Mesh Uniform Scale=0.4
    Sprite Size Mode=::NewEnumerator3
    Uniform Sprite Size=20
    UsePositionOffset=false
    Write Material Random=true
   )
   &ShapeLocation(MaskSpawnGroup=false, Override Seed=false, UseOwnerScale=true) {enabled=false}
   /SetVariables(Particles.Color001:LinearColor=<rgba 1 0 0 1>) {isSetParameters=true, module=SetVariables_#1}
   &InitialMeshOrientation(Use Rotation Vector=false)
  ]
  particleUpdateScript [
   /SetVariables() {isSetParameters=true, module=SetVariables_#2}
   &ParticleState(Kill Particles When Lifetime Has Elapsed=false)
   &self:FollowUpUpdate_01(Dir:Vector3f=<200 0 0>) {module=FollowUpUpdate_01}
   &CurlNoiseForce(Pan Noise=false) {enabled=false}
   &VortexForce(Origin Pull Amount=100, UseOriginOffset=false) {enabled=false}
   &Drag(Use Linear Drag=true, Use Rotational Drag=true)
   &PointForce(Force Strength=-1297.3162, Module.Use Falloff=false, Module.UseLinearForce=false) {enabled=false}
   &SolveForcesAndVelocity(Clamp Velocity=false, Limit Acceleration=false)
  ]
  rendererClasses [&NiagaraSpriteRendererProperties, &NiagaraMeshRendererProperties]
  renderers [
   {rendererClass=&NiagaraSpriteRendererProperties}
   {rendererClass=&NiagaraMeshRendererProperties, rendererIndex=1}
   {rendererClass=&NiagaraSpriteRendererProperties, rendererIndex=2}
   {rendererClass=&NiagaraMeshRendererProperties, rendererIndex=3}
  ]
 }
 Empty002 {
  bEnabled=false
  emitterUpdateScript [
   &EmitterState(Life Cycle Mode=::NewEnumerator1, Loop Behavior=::NewEnumerator1, Loop Duration=1, UseLoopDelay=false)
   &SpawnBurst_Instantaneous(Spawn Count=10, Use Spawn Probability=false, UseLoopCountLimit=false)
   &SpawnRate(SpawnRate=10, Use Spawn Probability=false) {enabled=false}
  ]
  particleSpawnScript [
   &InitializeParticle(
    Color Mode=::NewEnumerator0
    Lifetime=5
    Mesh Scale Mode=::NewEnumerator3
    Mesh Uniform Scale=0.4
    Sprite Size Mode=::NewEnumerator3
    Uniform Sprite Size=20
    UsePositionOffset=false
    Write Material Random=true
   )
   &ShapeLocation(MaskSpawnGroup=false, Override Seed=false, UseOwnerScale=true) {enabled=false}
   /SetVariables(Particles.Color001:LinearColor=<rgba 1 0 0 1>, Particles.Color002=<rgba 0 1 0.0311 1>) {isSetParameters=true, module=SetVariables_#1}
   &InitialMeshOrientation(Use Rotation Vector=false)
  ]
  particleUpdateScript [
   /SetVariables() {isSetParameters=true, module=SetVariables_#2}
   &ParticleState(Kill Particles When Lifetime Has Elapsed=false)
   &self:FollowUpUpdate_02(Dir:Vector3f=<200 0 0>, Hardness=0.1) {module=FollowUpUpdate_02}
   &CurlNoiseForce(Pan Noise=false) {enabled=false}
   &VortexForce(Origin Pull Amount=100, UseOriginOffset=false) {enabled=false}
   &Drag(Use Linear Drag=true, Use Rotational Drag=true)
   &PointForce(Force Strength=-1297.3162, Module.Use Falloff=false, Module.UseLinearForce=false) {enabled=false}
   &SolveForcesAndVelocity(Clamp Velocity=false, Limit Acceleration=false)
  ]
  rendererClasses [&NiagaraSpriteRendererProperties, &NiagaraMeshRendererProperties]
  renderers [
   {rendererClass=&NiagaraSpriteRendererProperties}
   {rendererClass=&NiagaraMeshRendererProperties, rendererIndex=1}
   {rendererClass=&NiagaraSpriteRendererProperties, rendererIndex=2}
   {rendererClass=&NiagaraMeshRendererProperties, rendererIndex=3}
   {rendererClass=&NiagaraSpriteRendererProperties, rendererIndex=4}
  ]
 }
 Empty003 {
  bEnabled=false
  emitterUpdateScript [
   &EmitterState(Life Cycle Mode=::NewEnumerator1, Loop Behavior=::NewEnumerator1, Loop Duration=1, UseLoopDelay=false)
   &SpawnBurst_Instantaneous(Spawn Count=12, Use Spawn Probability=false, UseLoopCountLimit=false)
   &SpawnRate(SpawnRate=10, Use Spawn Probability=false) {enabled=false}
  ]
  particleSpawnScript [
   &InitializeParticle(
    Color Mode=::NewEnumerator0
    Lifetime=5
    Mesh Scale Mode=::NewEnumerator3
    Mesh Uniform Scale=0.4
    Sprite Size Mode=::NewEnumerator3
    Uniform Sprite Size=20
    UsePositionOffset=false
    Write Material Random=true
   )
   &ShapeLocation(MaskSpawnGroup=false, Override Seed=false, UseOwnerScale=true) {enabled=false}
   /SetVariables(Particles.Color001:LinearColor=<rgba 1 0 0 1>, Particles.Color002=<rgba 0 1 0.0311 1>) {isSetParameters=true, module=SetVariables_#1}
   &InitialMeshOrientation(Use Rotation Vector=false)
  ]
  particleUpdateScript [
   /SetVariables() {isSetParameters=true, module=SetVariables_#2}
   &ParticleState(Kill Particles When Lifetime Has Elapsed=false)
   &self:FollowUpUpdate_03(Dir:Vector3f=<200 0 0>, Hardness:NiagaraFloat=&RandomRangeFloat(Minimum=0.944, Maximum=2.1705)) {module=FollowUpUpdate_03}
   &CurlNoiseForce(Pan Noise=false) {enabled=false}
   &VortexForce(Origin Pull Amount=100, UseOriginOffset=false) {enabled=false}
   &Drag(Use Linear Drag=true, Use Rotational Drag=true)
   &PointForce(Force Strength=-1297.3162, Module.Use Falloff=false, Module.UseLinearForce=false) {enabled=false}
   &SolveForcesAndVelocity(Clamp Velocity=false, Limit Acceleration=false)
  ]
  rendererClasses [&NiagaraSpriteRendererProperties, &NiagaraMeshRendererProperties]
  renderers [
   {rendererClass=&NiagaraSpriteRendererProperties}
   {rendererClass=&NiagaraMeshRendererProperties, rendererIndex=1}
   {rendererClass=&NiagaraSpriteRendererProperties, rendererIndex=2}
   {rendererClass=&NiagaraMeshRendererProperties, rendererIndex=3}
   {rendererClass=&NiagaraSpriteRendererProperties, rendererIndex=4}
  ]
 }
 Empty004 {
  emitterUpdateScript [
   &EmitterState(Life Cycle Mode=::NewEnumerator1, Loop Behavior=::NewEnumerator1, Loop Duration=1, UseLoopDelay=false)
   &SpawnBurst_Instantaneous(Spawn Count=24, Use Spawn Probability=false, UseLoopCountLimit=false)
   &SpawnRate(SpawnRate=10, Use Spawn Probability=false) {enabled=false}
  ]
  particleSpawnScript [
   &InitializeParticle(
    Color=<rgba 1 1 1 15.5921>
    Lifetime=5
    Mesh Scale Mode=::NewEnumerator3
    Mesh Uniform Scale=0.4
    Sprite Size Mode=::NewEnumerator3
    Uniform Sprite Size=20
    UsePositionOffset=false
    Write Material Random=true
   )
   &ShapeLocation(MaskSpawnGroup=false, Override Seed=false, UseOwnerScale=true) {enabled=false}
   /SetVariables(Particles.Color001:LinearColor=<rgba 1 0 0 1>, Particles.Color002=<rgba 0 1 0.0311 1>) {isSetParameters=true, module=SetVariables_#1}
   &InitialMeshOrientation(Use Rotation Vector=false)
  ]
  particleUpdateScript [
   /SetVariables() {isSetParameters=true, module=SetVariables_#2}
   &ParticleState(Kill Particles When Lifetime Has Elapsed=false)
   &/NM_PendulumSimulation(Hardness:NiagaraFloat=&RandomRangeFloat(Minimum=1, Maximum=2), TailPivot:Vector3f=<200 0 0>)
   &CurlNoiseForce(Pan Noise=false) {enabled=false}
   &VortexForce(Origin Pull Amount=100, UseOriginOffset=false) {enabled=false}
   &Drag(Use Linear Drag=true, Use Rotational Drag=true)
   &PointForce(Force Strength=-1297.3162, Module.Use Falloff=false, Module.UseLinearForce=false) {enabled=false}
   &SolveForcesAndVelocity(Clamp Velocity=false, Limit Acceleration=false)
  ]
  rendererClasses [&NiagaraSpriteRendererProperties, &NiagaraMeshRendererProperties]
  renderers [
   {rendererClass=&NiagaraSpriteRendererProperties}
   {rendererClass=&NiagaraMeshRendererProperties, rendererIndex=1}
   {rendererClass=&NiagaraSpriteRendererProperties, rendererIndex=2}
   {rendererClass=&NiagaraMeshRendererProperties, rendererIndex=3}
   {rendererClass=&NiagaraSpriteRendererProperties, rendererIndex=4}
  ]
 }
 Empty005 {
  emitterUpdateScript [
   &EmitterState(Life Cycle Mode=::NewEnumerator1, Loop Behavior=::NewEnumerator1, Loop Duration=1, UseLoopDelay=false)
   &SpawnBurst_Instantaneous(Spawn Count=24, Use Spawn Probability=false, UseLoopCountLimit=false)
   &SpawnRate(SpawnRate=10, Use Spawn Probability=false) {enabled=false}
  ]
  particleSpawnScript [
   &InitializeParticle(
    Color=<rgba 5 5 5 1>
    Lifetime=5
    Mesh Scale Mode=::NewEnumerator3
    Mesh Uniform Scale=0.4
    Sprite Size Mode=::NewEnumerator3
    Uniform Sprite Size=20
    UsePositionOffset=false
    Write Material Random=true
   )
   &ShapeLocation(MaskSpawnGroup=false, Override Seed=false, UseOwnerScale=true) {enabled=false}
   /SetVariables(Particles.Color001:LinearColor=<rgba 1 0 0 1>, Particles.Color002=<rgba 0 1 0.0311 1>) {isSetParameters=true, module=SetVariables_#1}
   &InitialMeshOrientation(Use Rotation Vector=false)
  ]
  particleUpdateScript [
   /SetVariables() {isSetParameters=true, module=SetVariables_#2}
   &ParticleState(Kill Particles When Lifetime Has Elapsed=false)
   &NM_PendulumSimulation(Hardness:NiagaraFloat=&RandomRangeFloat(Minimum=1, Maximum=1), TailPivot:Vector3f=<200 0 0>)
   &CurlNoiseForce(Pan Noise=false) {enabled=false}
   &VortexForce(Origin Pull Amount=100, UseOriginOffset=false) {enabled=false}
   &Drag(Use Linear Drag=true, Use Rotational Drag=true)
   &PointForce(Force Strength=-1297.3162, Module.Use Falloff=false, Module.UseLinearForce=false) {enabled=false}
   &SolveForcesAndVelocity(Clamp Velocity=false, Limit Acceleration=false)
   &DynamicMaterialParameters(
    Index 0 Param 1=&Add_Float(
     A=&RandomRangeFloat(Minimum=0, Maximum=111)
     B=&Multiply_Float(A=&RandomRangeFloat(Minimum=0.0947, Maximum=0.5627), B=$Engine.Time)
    )
    Module.Index0Param0Write=true
    Module.Index0Param1Write=true
    Module.Index0Param2Write=true
    Module.Index0Param3Write=true
   )
  ]
  rendererClasses [&NiagaraSpriteRendererProperties, &NiagaraMeshRendererProperties]
  renderers [
   {rendererClass=&NiagaraSpriteRendererProperties}
   {rendererClass=&NiagaraMeshRendererProperties, rendererIndex=1}
   {rendererClass=&NiagaraSpriteRendererProperties, rendererIndex=2}
   {rendererClass=&NiagaraMeshRendererProperties, rendererIndex=3}
   {rendererClass=&NiagaraSpriteRendererProperties, rendererIndex=4}
  ]
 }
}
summary {
 systemName=NS_PendulumSimulation_RD
 emitters [
  {
   bEnabled=false
   emitterName=Empty
   rendererClasses [&NiagaraSpriteRendererProperties, &NiagaraMeshRendererProperties]
  }
  {
   bEnabled=false
   emitterName=Empty001
   rendererClasses [&NiagaraSpriteRendererProperties, &NiagaraMeshRendererProperties]
  }
  {
   bEnabled=false
   emitterName=Empty002
   rendererClasses [&NiagaraSpriteRendererProperties, &NiagaraMeshRendererProperties]
  }
  {
   bEnabled=false
   emitterName=Empty003
   rendererClasses [&NiagaraSpriteRendererProperties, &NiagaraMeshRendererProperties]
  }
  {emitterName=Empty004, rendererClasses [&NiagaraSpriteRendererProperties, &NiagaraMeshRendererProperties]}
  {emitterName=Empty005, rendererClasses [&NiagaraSpriteRendererProperties, &NiagaraMeshRendererProperties]}
 ]
 userVariables []
}
